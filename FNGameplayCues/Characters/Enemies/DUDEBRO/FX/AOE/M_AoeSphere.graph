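# Translucent unlit, two-sided particle-mesh material for an AOE sphere effect in the DUDEBRO enemy's gameplay cue.
# Curve-atlas ramps shape opacity from camera distance, view angle, noisy depth fade and UV gradient, with vertex color masking and optional Fresnel fade.
# Swirl-distorted panning noise drives a color ramp, blended with offset scene color on the default shading path and tinted by particle color.
# Dynamic parameters control noise time, tiling and vertex offset, while opacity drives refraction.
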Material /FNGameplayCues/Characters/Enemies/DUDEBRO/FX/AOE/M_AoeSphere
graph {
 properties {
  bUseMaterialAttributes=true
  bUsedWithMeshParticles=true
  bUsedWithNiagaraMeshParticles=true
  blendMode=BLEND_Translucent
  pixelDepthOffsetMode=PDOM_Legacy
  refractionMethod=RM_IndexOfRefraction
  shadingModel=MSM_Unlit
  twoSided=true
  usedShadingModels=Unlit
 }
 nodes [
  ScalarParameter() {defaultValue=-20.9964, parameterName=WPO}
  n1 = VertexColor()
  n2 = DynamicParameter() {defaultValue=<rgba 1 1 1 20>, paramNames [Time,U,V,WPO]}
  n3 = Multiply(A=Transform(None=Constant3Vector() {constant=<rgba 1 0 0 0>}), B=Multiply(A=LinearInterpolate(Alpha=@n1.G) {constA=0.85}, B=@n2.WPO)) {constB=100}
  n4 = CurveAtlasRowParameter(CurveTime=SphereMask(A=WorldPosition().XYZ, B=CameraPositionWS(), Radius=Constant() {r=2048}, Hardness=Constant())) {
   atlas=&AthenaCurveAtlas01
   curve=/Game/Athena/LinearColorCurves/Curve_StormKingRamp1
   defaultValue=28
   parameterName=Curve
  }
  n5 = WorldPosition()
  n6 = ScalarParameter() {parameterName=Uses Particle Position}
  n7 = ParticlePositionWS()
  n8 = ObjectPositionWS()
  n9 = Normalize(VectorInput=Subtract(A=@n5.XYZ, B=If(A=@n6, "A > B"=@n7, "A == B"=@n7, "A < B"=@n8) {constB=0.99}))
  n10 = Normalize(VectorInput=Subtract(A=@n5.XYZ, B=CameraPositionWS()))
  n11 = CurveAtlasRowParameter(CurveTime=Abs(None=DotProduct(A=@n9, B=@n10))) {
   atlas=&AthenaCurveAtlas01
   curve=&Curve_StormKingRamp1
   defaultValue=28
   parameterName=Curve
  }
  n12 = Multiply(A=TextureCoordinate() {uTiling=6, vTiling=6}, B=AppendVector(A=@n2.U, B=@n2.U))
  n13 = TextureSample(UVs=Add(A=Panner(Coordinate=@n12, Time=@n2.Time) {speedX=0.5, speedY=0.25}, B=VertexColor().G)) {
   samplerSource=SSM_Wrap_WorldGroupSettings
   samplerType=SAMPLERTYPE_LinearColor
   texture=&T_NoisePack_LinearColor
  }
  n14 = LinearInterpolate(Alpha=@n13.G) {constA=0.5, constB=1.5}
  n15 = Multiply(A=PixelDepth().R) {constB=0.5}
  n16 = CurveAtlasRowParameter(CurveTime=DepthFade(FadeDistance=Multiply(A=@n14, B=@n15)) {fadeDistanceDefault=500}) {
   atlas=&AthenaCurveAtlas01
   curve=&Curve_StormKingRamp1
   defaultValue=28
   parameterName=Curve
  }
  n17 = CurveAtlasRowParameter(CurveTime=ComponentMask(None=TextureCoordinate()) {g=true}) {
   atlas=&AthenaCurveAtlas01
   curve=&Curve_StormKingRamp1
   defaultValue=28
   parameterName=Curve
  }
  n18 = Multiply(A=@n17.R, B=VertexColor().R)
  n19 = Multiply(A=@n16.G, B=@n18)
  n20 = Multiply(
   A=@n11.B
   B=ShadingPathSwitch(Default=ShadingPathSwitch(Default=@n19, Mobile=@n18) {desc=Migrated from FeatureLevelSwitch}, Forward=@n18, Mobile=@n18)
  )
  n21 = GetMaterialAttributes(None=SetMaterialAttributes!(
    MaterialAttributes=SetMaterialAttributes!(MaterialAttributes, World Position Offset=@n3) {attributeSetTypes [#1]}
    Opacity=Multiply(A=@n4.A, B=@n20)
   ) {attributeSetTypes [#2]}) {attributeGetTypes [#2]}
  n22 = DynamicParameter() {defaultValue=<rgba 1 1 1 20>, paramNames [Time,U,V,WPO]}
  n23 = TextureCoordinate() {uTiling=4, vTiling=4}
  n24 = DynamicParameter() {defaultValue=<rgba 1 1 1 20>, paramNames [Time,U,V,WPO]}
  n25 = AppendVector(A=@n24.U, B=@n24.U)
  n26 = TextureSample(UVs=Panner(Coordinate=Multiply(A=@n23, B=@n25), Time=@n24.Time) {speedX=0.05, speedY=0.25}) {samplerSource=SSM_Wrap_WorldGroupSettings, samplerType=SAMPLERTYPE_Normal, texture=/Game/Global/Textures/Default/Flowmaps/T_SwirlFlow_02_N}
  n27 = Multiply(A=Multiply(A=@n21.Opacity, B=ComponentMask(None=@n26.RGB) {g=true, r=true}), B=LinearInterpolate(Alpha=@n13.B) {constA=-0.3, constB=0.3}) {constB=0.4}
  n28 = Add(A=Multiply(A=TextureCoordinate() {uTiling=10, vTiling=10}, B=AppendVector(A=@n22.U, B=@n22.U)), B=@n27)
  n29 = TextureSample(UVs=Add(A=Panner(Coordinate=@n28, Time=@n22.Time) {speedX=0.5, speedY=0.1}, B=VertexColor().G)) {
   samplerSource=SSM_Wrap_WorldGroupSettings
   samplerType=SAMPLERTYPE_LinearColor
   texture=&T_NoisePack_LinearColor
  }
  n30 = CurveAtlasRowParameter(CurveTime=Multiply(A=@n29.R, B=@n21.Opacity)) {
   atlas=&AthenaCurveAtlas01
   curve=/Game/Athena/LinearColorCurves/Curve_StormKingRamp2
   defaultValue=29
   parameterName=Curve
  }
  n31 = ComponentMask(None=SceneColor!(OffsetFraction=Multiply(A=@n27) {constB=0.33}) {inputMode=OffsetFraction}.RGB) {b=true, g=true, r=true}
  n32 = ShadingPathSwitch(Default=LinearInterpolate(A=@n30, B=Multiply(A=@n31) {constB=2}) {constAlpha=0.25}, Mobile=@n30) {desc=Migrated from FeatureLevelSwitch}
  n33 = GetMaterialAttributes(None=SetMaterialAttributes!(MaterialAttributes=@n21.MaterialAttributes, Emissive Color=ShadingPathSwitch(Default=@n32, Forward=@n30, Mobile=@n30)) {attributeSetTypes [#3]}) {attributeGetTypes [#3, #2]}
  n34 = ParticleColor()
  n35 = ScalarParameter() {group=Opacity, parameterName=Angle Fade Power}
  n36 = OneMinus(None=Fresnel(ExponentIn=@n35, Normal=VertexNormalWS()))
  n37 = Constant() {r=1}
  n38 = SetMaterialAttributes!(
   MaterialAttributes=@n33.MaterialAttributes
   Emissive Color=Multiply(A=@n33.Emissive Color, B=@n34.RGB)
   Opacity=Multiply(A=Multiply(A=@n33.Opacity, B=@n34.A), B=If(A=@n35, "A > B"=@n36, "A == B"=@n36, "A < B"=@n37) {constB=0.01})
  ) {attributeSetTypes [#3, #2]}
  SetMaterialAttributes!(
   MaterialAttributes=@n38
   Normal
   Refraction=LinearInterpolate(Alpha=GetMaterialAttributes(None=@n38) {attributeGetTypes [#2]}.Opacity) {constA=1, constB=1.05}
  ) {attributeSetTypes [#4, #5]}
 ]
}
